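; Snowy big bush building actor that wobbles, tags players inside it and dies to driven vehicles.
; Server-enabled sphere overlaps add or remove cornfield and bush-tagger gameplay effects, reset fall height and drive curve-table-gated dynamic-material wobble from player velocity.
; A collision response swapper registers on that sphere.
; Construction and BeginPlay write seasonal colors to custom primitive data, where the color-option snow value overwrites Snow Coverage Amount in slot 6.
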(event EventBeginPlay
  ;[1 1 1] EVENT -> BeginPlay
  (Utilities|FlowControl|Sequence
    (:then_0
      (bind _outxy (DataTable|EvaluateCurveTableRow "/Game/Athena/Balance/DataTables/AthenaGameData.AthenaGameData" "Default.Cornfield.Wobble")
        (:RowFound
          (Variables|Default|SetShouldWobble (Math|Conversions|ToBoolean(Integer) (Math|Float|Truncate _outxy))))
        (:RowNotFound
          _))
      (if (Networking|IsServer)
        (Variables|/Script/Engine.PrimitiveComponent|SetbGenerateOverlapEvents true (Variables|Default|GetOverlapSphere))))
    (:then_1
      (Class|FortCollisionResponseSwapperComponent|RegisterCallbacksforOverlap (Variables|Default|GetFortCollisionResponseSwapper) (Variables|Debug|GetStaticMeshComponent) (Variables|Default|GetOverlapSphere)))
    (:then_2
      (Networking|CanExecuteCosmeticEvents
        (:True
          (Rendering|Material|SetCustomPrimitiveDataFloat (Variables|Debug|GetStaticMeshComponent) 6 (Variables|Default|GetSnow%20Coverage%20Amount))
          (Default|SetTreeSeasonalColor))))
    (:then_3)))

(event Custom|MaterialWobble
  (params (player :type Actor%20Object%20Reference))
  (bind _mid (Variables|Default|GetMID))
  ;[1 1 1] Material wobble
  (Networking|CanExecuteCosmeticEvents
    (:True
      (Utilities|IsValid (Variables|Default|GetMID)
        (:Is%20Valid
          _)
        (:Is%20Not%20Valid
          (bind _returnvalue (Rendering|Material|CreateDynamicMaterialInstance (Variables|Debug|GetStaticMeshComponent) 0 0 "None" :!class /Script/Engine.PrimitiveComponent))
          (Variables|Default|SetMID _returnvalue)))
      (bind _asplayer_pawn_athena (Utilities|Casting|CastToPlayerPawn_Athena player))
      (Variables|Default|SetPlayerVelocity (Transformation|GetVelocity _asplayer_pawn_athena))
      (Rendering|Material|SetVectorParameterValue _mid "HitLocation" (Math|Conversions|ToLinearColor(Vector) (Transformation|GetActorLocation _asplayer_pawn_athena)) :!class /Script/Engine.MaterialInstanceDynamic)
      (Rendering|Material|SetVectorParameterValue _mid "WobbleDirection" (Math|Conversions|ToLinearColor(Vector) (Math|Vector|Normalize (Variables|Default|GetPlayerVelocity))) :!class /Script/Engine.MaterialInstanceDynamic)
      (Rendering|Material|SetScalarParameterValue _mid "Time" (Utilities|Time|GetTimeSeconds) :!class /Script/Engine.MaterialInstanceDynamic)
      (Rendering|Material|SetScalarParameterValue _mid "WobbleStrength" (select (> (Math|Vector|VectorLength (Variables|Default|GetPlayerVelocity)) 295.0) 3e+01 5.0) :!class /Script/Engine.MaterialInstanceDynamic))))

(event OnComponentBeginOverlap(OverlapSphere)
  (params (OverlappedComponent :type Primitive%20Component%20Object%20Reference) (OtherActor :type Actor%20Object%20Reference) (OtherComp :type Primitive%20Component%20Object%20Reference) (OtherBodyIndex :type Integer) (bFromSweep :type Boolean) (SweepResult :type Hit%20Result%20Structure%20%28by%20ref%29))
  ;[1 1 1] Whenever you overlap a cornfield rectangle, you get this GE unless you already have it
  (bind _asplayer_pawn_athena (Utilities|Casting|CastToPlayerPawn_Athena OtherActor)
    (:then
      (if (Variables|Default|GetShouldWobble)
        (Networking|FlushNetDormancy)
        (CallFunction|MaterialWobble _asplayer_pawn_athena)
        (else
          _))
      (Utilities|IsValid (Ability|GetAbilitySystemComponent _asplayer_pawn_athena)
        (:Is%20Valid
          (GameplayEffects|ApplyGameplayEffectToTarget :self (Ability|GetAbilitySystemComponent _asplayer_pawn_athena) :GameplayEffectClass (Variables|Default|GetGE_Cornfield_GC) :Target (Ability|GetAbilitySystemComponent _asplayer_pawn_athena) :Level 1.0)
          (GameplayEffects|ApplyGameplayEffectToTarget :self (Ability|GetAbilitySystemComponent _asplayer_pawn_athena) :GameplayEffectClass (Variables|Default|GetGE%20BushTagger) :Target (Ability|GetAbilitySystemComponent _asplayer_pawn_athena) :Level 1.0)
          (Networking|ForceNetUpdate _asplayer_pawn_athena)
          (Pawn|Character|ResetFallingHeight _asplayer_pawn_athena))))
    (:CastFailed
      ;[1 1 1] Destroy when cornfield comes in contact with a vehicle
      (bind _asfort_vehicle_interface (Utilities|Casting|CastToFortVehicleInterface OtherActor))
      (Utilities|IsValid (Vehicle|GetDriver _asfort_vehicle_interface)
        (:Is%20Valid
          (Gameplay|Die :Killer (Pawn|GetController (Vehicle|GetDriver _asfort_vehicle_interface)) :DamageCauser (Vehicle|GetDriver _asfort_vehicle_interface)))))))

(event OnComponentEndOverlap(OverlapSphere)
  (params (OverlappedComponent :type Primitive%20Component%20Object%20Reference) (OtherActor :type Actor%20Object%20Reference) (OtherComp :type Primitive%20Component%20Object%20Reference) (OtherBodyIndex :type Integer))
  ;[1 1 1] Whenever you stop overlapping a cornfield, remove the GE and the gameplay cue
  (bind _asplayer_pawn_athena (Utilities|Casting|CastToPlayerPawn_Athena OtherActor))
  (GameplayEffects|RemoveActiveGameplayEffectbySourceEffect (Ability|GetAbilitySystemComponent _asplayer_pawn_athena) (Variables|Default|GetGE_Cornfield_GC) 0 1)
  (GameplayEffects|RemoveActiveGameplayEffectbySourceEffect (Ability|GetAbilitySystemComponent _asplayer_pawn_athena) (Variables|Default|GetGE%20BushTagger) 0 1)
  (Networking|ForceNetUpdate _asplayer_pawn_athena))

(event EventEndPlay
  (params (EndPlayReason :type EEndPlayReason%20Enum :default "Destroyed"))
  ;[1 1 1] EVENT -> EndPlay
  (Class|FortCollisionResponseSwapperComponent|UnregisterCallbacksforOverlap (Variables|Default|GetFortCollisionResponseSwapper)))

(fn SetSeasonPresetColors
  (bind _array_element (Utilities|Array|ForEachLoopwithBreak (Variables|/Game/ArtTools/Procedural/Trees/TreeSeasonColor_Data.TreeSeasonColor_Data_C|GetTreeSeasonData (Variables|LeaveColor|GetTreeSeasonColor%20Data)))
    (:LoopBody
      (if (Utilities|Enum|Equal(Enum) (Utilities|Struct|BreakTreeSeasonColorStruct _array_element) (Variables|LeaveColor|GetTree%20Color%20Option))
        (Variables|LeaveColor|SetColor (.Color_5_E33006414CF6D999D4D482A4D2FF03F6 (Utilities|Struct|BreakTreeSeasonColorStruct _array_element)))
        (Variables|LeaveColor|SetHue_Shift_BlueVertexColor (.ValueVariation_9_4FAED9E04DA8955B0BC4EDB1E0A3C1F7 (Utilities|Struct|BreakTreeSeasonColorStruct _array_element)))
        (Variables|LeaveColor|SetHueVariation_PerLeaf (.HueVariation_10_00BFFB9140C4AD7D16C9AA9AEB01A935 (Utilities|Struct|BreakTreeSeasonColorStruct _array_element)))))))

(fn SetTreeSeasonalColor
  (Utilities|FlowControl|Sequence
    (:then_0
      (if (Utilities|Enum|NotEqual(Enum) (Variables|LeaveColor|GetTree%20Color%20Option) "NewEnumerator0")
        (Default|SetSeasonPresetColors)))
    (:then_1
      (Default|UpdateColor))))

(fn UpdateColor
  (bind (_r _g _b _a) (Utilities|Struct|BreakLinearColor (Variables|LeaveColor|GetColor)))
  (bind (_r_2 _g_2 _b_2) (Utilities|Struct|BreakLinearColor (Variables|LeaveColor|GetColor_2)))
  (bind _staticmeshcomponent (Variables|Debug|GetStaticMeshComponent))
  (Rendering|Material|SetCustomPrimitiveDataVector4 _staticmeshcomponent 0 (Math|Vector4|MakeVector4 _r _g _b _a))
  (Rendering|Material|SetCustomPrimitiveDataVector3 _staticmeshcomponent 17 (Math|Vector|MakeVector _r_2 _g_2 _b_2))
  (Rendering|Material|SetCustomPrimitiveDataFloat (Variables|Debug|GetStaticMeshComponent) 4 (Variables|LeaveColor|GetHueVariation_PerLeaf))
  (Rendering|Material|SetCustomPrimitiveDataFloat (Variables|Debug|GetStaticMeshComponent) 5 (Variables|LeaveColor|GetHue_Shift_BlueVertexColor))
  (Rendering|Material|SetCustomPrimitiveDataFloat (Variables|Debug|GetStaticMeshComponent) 6 (Utilities|Select :Index (Variables|LeaveColor|GetTree%20Color%20Option) :NewEnumerator0 0.0 :NewEnumerator1 0.0 :NewEnumerator2 0.0 :NewEnumerator3 0.0 :NewEnumerator4 1.0)))

(fn ConstructionScript
  (Networking|CanExecuteCosmeticEvents
    (:True
      (Rendering|Material|SetCustomPrimitiveDataFloat (Variables|Debug|GetStaticMeshComponent) 6 (Variables|Default|GetSnow%20Coverage%20Amount))
      (Default|SetTreeSeasonalColor))))
