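; Thrown variant of the Apollo big bush foliage actor, spawned for the Big Bush Grenade.
; On server BeginPlay it links to the actor below, destroys overlapping big bushes via a box query, then calls the scale-on-spawn event.
; That event unhides the bush and, off-server only, scales the mesh by timeline, then triggers a dynamic-material wobble.
; The bush destroys itself when its linked actor is destroyed.
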(event Collision|EventActorBeginOverlap
  (params (OtherActor :type Actor%20Object%20Reference)))

(event EventEndPlay
  (params (EndPlayReason :type EEndPlayReason%20Enum :default "Destroyed"))
  ;[1 1 1] EVENT -> EndPlay
  (Class|FortCollisionResponseSwapperComponent|UnregisterCallbacksforOverlap (Variables|Default|GetFortCollisionResponseSwapper)))

(event EventBeginPlay
  (|CallParentFunction)
  (if (Networking|IsServer)
    (LinkToActorComponent|PerformLinkingCheck (Variables|Default|GetFortLinkToActor) "Down" (Variables|Default|GetLinkedActor))
    ;[0.2688 0.3094 0] Destroy other Big Bushes if we're touching them ( for Big Bush Grenade )
    (bind _outactors (Collision|BoxOverlapActors (Transformation|GetWorldLocation (Variables|Default|GetOverlapBox)) (Variables|/Script/Engine.BoxComponent|GetBoxExtent (Variables|Default|GetOverlapBox)) (Variables|Default|GetObject%20Types) "/Game/Athena/Apollo/Environments/BuildingActors/Foliage/Apollo_BigBush.Apollo_BigBush_C" (Utilities|Array|MakeArray self) :!class /Script/Engine.KismetSystemLibrary))
    (for _array_element _outactors
      (bind _asapollo_big_bush (Utilities|Casting|CastToApollo_BigBush _array_element))
      (Actor|DestroyActor _asapollo_big_bush))
    (CallFunction|ScaleonTImeline)))

(event EventTick
  (params (DeltaSeconds :type Float%20%28single-precision%29)))

(event Custom|ScaleOnTImeline
  ;[1 1 1] Scale over Timeline on spawn
  (Variables|Default|SetBaseScale_0 (Transformation|GetActorScale3D self))
  (Rendering|SetActorHiddenInGame false)
  (if (not (Networking|IsServer))
    (bind (_direction _newtrack_0) (Timeline|Timeline_0 PlayFromStart)
      (:Update
        (Transformation|SetWorldScale3D (Variables|Debug|GetStaticMeshComponent) (* _newtrack_0 (Variables|Default|GetBaseScale_0))))
      (:Finished
        (CallFunction|WobbleonSpawn)))))

(event Custom|WobbleOnSpawn
  (bind _returnvalue (Math|Vector|Normalize "2000.000000,2000.000000,0.000000"))
  (bind _mid (Variables|Default|GetMID))
  ;[1 1 1] Material wobble
  (Networking|CanExecuteCosmeticEvents
    (:True
      (Utilities|IsValid (Variables|Default|GetMID)
        (:Is%20Valid
          _)
        (:Is%20Not%20Valid
          (bind _returnvalue_2 (Rendering|Material|CreateDynamicMaterialInstance (Variables|Debug|GetStaticMeshComponent) 0 0 "None" :!class /Script/Engine.PrimitiveComponent))
          (Variables|Default|SetMID _returnvalue_2)))
      (Rendering|Material|SetVectorParameterValue _mid "HitLocation" (Math|Conversions|ToLinearColor(Vector) (Transformation|GetActorLocation)) :!class /Script/Engine.MaterialInstanceDynamic)
      (Rendering|Material|SetVectorParameterValue _mid "WobbleDirection" (Math|Conversions|ToLinearColor(Vector) _returnvalue) :!class /Script/Engine.MaterialInstanceDynamic)
      (Rendering|Material|SetScalarParameterValue _mid "Time" (Utilities|Time|GetTimeSeconds) :!class /Script/Engine.MaterialInstanceDynamic)
      (Rendering|Material|SetScalarParameterValue _mid "WobbleStrength" (select (> (Math|Vector|VectorLength _returnvalue) 295.0) 3e+01 5.0) :!class /Script/Engine.MaterialInstanceDynamic))))

(event OnLinkedActorDestroyed(FortLinkToActor)
  (params (DamageCauser :type Actor%20Object%20Reference))
  (Actor|DestroyActor))

(fn ConstructionScript
  (|CallParentFunction))
