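# Niagara system for autumn bush destruction, converted from Cascade, with warm-tinted mesh leaves and dark sprite chunks.
# Leaf bursts once per loop, scattering leaves that tumble, orbit via Cascade-conversion modules, fall gently, scale and fade.
# DarkChunksFast gives short-lived sprites heavy gravity and speed-based stretching, but shows no spawn module.
# Branch_Lightweight and ChunksLazy_LW list no modules or renderers here.
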NiagaraSystem /Game/Athena/Apollo/Environments/BuildingActors/Foliage/NS_LG_AutumnBush_Destruction_Converted
dependencies {
 usedDataInterfaces [
  &NiagaraDataInterfaceCurve
  &NiagaraDataInterfaceVector2DCurve
  &NiagaraDataInterfaceVectorCurve
 ]
 usedDynamicInputs []
 usedModules [&SystemState]
}
emitters {
 Branch_Lightweight {rendererClasses [], renderers []}
 ChunksLazy_LW {rendererClasses [], renderers []}
 DarkChunksFast {
  emitterUpdateScript [&EmitterState()]
  particleSpawnScript [
   &InitializeParticle(
    Color=<rgba 0.001 0.001 0.001 1>
    Lifetime=&RandomRangeFloat(Minimum=0.1, Maximum=1.2, Evaluation Type=::NewEnumerator1)
    Position Offset=&RandomRangeVector(Minimum=<-256 -256 -50>, Maximum=<256 256 200>, Evaluation Type=::NewEnumerator1)
    Sprite Size=&RandomRangeVector2D(Minimum=<4 4>, Maximum=<8 8>, Evaluation Type=::NewEnumerator1)
    Sprite Size Mode=::NewEnumerator4
    UsePositionOffset=true
    Write Material Random=true
   )
   &AddVelocity(Velocity=&RandomRangeVector(Minimum=<-75 -75 0>, Maximum=<75 75 650>, Evaluation Type=::NewEnumerator1))
   &ShapeLocation(MaskSpawnGroup=false, Override Seed=false, Sphere Radius=12, UseOwnerScale=true)
   /CascadeToNiagaraConverter/NiagaraScripts/ModuleScripts/FixShapeVector()
   &AddVelocity(Velocity=&Multiply_VectorByFloat(
     Vector=$Output.FixShapeVector.ShapeVector
     Float=&RandomRangeFloat(Minimum=12, Maximum=3, Evaluation Type=::NewEnumerator1)
    )) {module=AddVelocity001}
  ]
  particleUpdateScript [
   &ParticleState()
   &AccelerationForce(Acceleration=<0 0 -2000>)
   &SolveForcesAndVelocity(Clamp Velocity=false, Limit Acceleration=false)
   &ScaleSpriteSize(
    Scale Factor=&Vector2DFromCurve(Vector2Curve=DataInterface() {propertyValues {XCurve {keys [<at 0.8 1>,<at 1 0>]}, YCurve {keys [<at 0.8 1>,<at 1 0>]}}}, ScaleCurve=<1 1>)
    Scale Sprite Size Mode=::NewEnumerator2
   )
   &ScaleSpriteSizeBySpeed(
    Max Scale Factor=<1 3>
    Min Scale Factor=<0.01 0.01>
    Source Velocity:Vector3f=&Multiply_Vector(A=<0.01 0.01 1>, B=$Particles.Velocity)
   )
   &ScaleColor(
    Scale Alpha=&FloatFromCurve(FloatCurve=DataInterface() {propertyValues {Curve {keys [<at 0 3>,<at 0.2 1>]}}})
    Scale RGB=&VectorFromCurve(VectorCurve=DataInterface() {propertyValues {
       XCurve {keys [<at 0.5 12 RCIM_Cubic>,<at 1 1 RCIM_Cubic>]}
       YCurve {keys [<at 0.5 6 RCIM_Cubic>,<at 1 1 RCIM_Cubic>]}
       ZCurve {keys [<at 0.5 2 RCIM_Cubic>,<at 1 1 RCIM_Cubic>]}
      }})
    ScaleA=true
    ScaleRGB=true
   )
  ]
  rendererClasses [&NiagaraSpriteRendererProperties]
  renderers [{rendererClass=&NiagaraSpriteRendererProperties}]
 }
 Leaf {
  emitterUpdateScript [
   &EmitterState()
   &SpawnBurst_Instantaneous(
    Spawn Count=&RandomRangeInt(Minimum=12, Maximum=16)
    Use Spawn Probability=false
    UseLoopCountLimit=false
   )
  ]
  particleSpawnScript [
   &InitializeParticle(
    Color Maximum=<rgba 6 1 0.5 1>
    Color Minimum=<rgba 8 6 1 1>
    Color Mode=::NewEnumerator2
    Lifetime=&RandomRangeFloat(Minimum=1.5, Maximum=3, Evaluation Type=::NewEnumerator1)
    Mesh Scale=&RandomRangeVector(Minimum=<0.4 0.4 0.4>, Maximum=<0.8 0.8 0.8>, Evaluation Type=::NewEnumerator1)
    Mesh Scale Mode=::NewEnumerator4
    Position Offset=&RandomRangeVector(Minimum=<-235 -235 0>, Maximum=<235 235 200>, Evaluation Type=::NewEnumerator1)
    UsePositionOffset=true
    Write Material Random=true
   )
   &AddVelocity(Velocity=&RandomRangeVector(Minimum=<-50 -50 75>, Maximum=<50 50 325>, Evaluation Type=::NewEnumerator1))
   &ShapeLocation(MaskSpawnGroup=false, Override Seed=false, Sphere Radius=30, UseOwnerScale=true)
   &FixShapeVector()
   &AddVelocity(Velocity=&Multiply_VectorByFloat(
     Vector=$Output.FixShapeVector.ShapeVector
     Float=&RandomRangeFloat(Minimum=1, Maximum=4, Evaluation Type=::NewEnumerator1)
    )) {module=AddVelocity001}
   &CameraOffset(Camera Offset Amount=&RandomRangeFloat(Minimum=40, Maximum=100, Evaluation Type=::NewEnumerator1))
   &InitialMeshOrientation(Use Rotation Vector=false)
  ]
  particleUpdateScript [
   &ParticleState()
   &ScaleMeshSize(Scale Factor=&VectorFromFloat(Value=&FloatFromCurve(FloatCurve=DataInterface() {propertyValues {Curve {keys [
          <at 0 0>
          {arriveTangent=1.4056, interpMode=RCIM_Cubic, leaveTangent=1.4056, time=0.0663, value=1.0008}
          {arriveTangent=-1.0719, interpMode=RCIM_Cubic, leaveTangent=-1.0719, time=0.7209, value=1.0133}
          <at 1 0>
         ]}}})))
   &AccelerationForce(Acceleration=&RandomRangeVector(Minimum=<0 0 -400>, Maximum=<0 50 -150>, Evaluation Type=::NewEnumerator1))
   &SolveForcesAndVelocity(Clamp Velocity=false, Limit Acceleration=false)
   /CascadeToNiagaraConverter/NiagaraScripts/ModuleScripts/UpdateMeshOrientationByEuler(RotationRate:Vector3f=&RandomRangeVector(Minimum=<-1 -1 -1>, Maximum=<1.5 1.5 1.5>))
   /Niagara/Modules/CascadeConversion/CascadeConversion_Orbit(
    Offset:Vector3f=&RandomRangeVector(Minimum=<10 10 10>, Maximum=<30 0 30>)
    Rotation:Vector3f=&RandomRangeVector(Minimum=<0 0 0>, Maximum=<1 1 1>)
    Rotation Rate:Vector3f=&RandomRangeVector(Minimum=<0.2 0.2 0.2>, Maximum=<0.5 0.5 0.5>)
   )
   /Niagara/Modules/CascadeConversion/CascadeConversion_SolveOrbit()
   &ScaleColor(
    Scale Alpha=&FloatFromCurve(FloatCurve=DataInterface() {propertyValues {Curve {keys [<at 0 1>,<at 1 0>]}}})
    ScaleA=true
    ScaleRGB=true
   )
  ]
  rendererClasses [&NiagaraMeshRendererProperties]
  renderers [{rendererClass=&NiagaraMeshRendererProperties, rendererIndex=0}]
 }
}
summary {
 systemName=NS_LG_AutumnBush_Destruction_Converted
 emitters [
  {emitterName=Leaf, rendererClasses [&NiagaraMeshRendererProperties]}
  {emitterName=Branch_Lightweight, rendererClasses []}
  {emitterName=ChunksLazy_LW, rendererClasses []}
  {emitterName=DarkChunksFast, rendererClasses [&NiagaraSpriteRendererProperties]}
 ]
 userVariables []
}
